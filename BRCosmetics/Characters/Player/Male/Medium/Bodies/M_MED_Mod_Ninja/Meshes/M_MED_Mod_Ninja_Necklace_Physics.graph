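# Physics asset for the necklace on the medium male Ninja player body, setting up its dynamic bones for simulation.
# A large spine_05 capsule stands in for the upper torso and anchors a chain of three necklace capsules.
# Two tag capsules both branch from the last necklace link.
# Every joint locks twist, and the tag joints also lock one swing axis, restricting them to a single swing direction.
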PhysicsAsset /BRCosmetics/Characters/Player/Male/Medium/Bodies/M_MED_Mod_Ninja/Meshes/M_MED_Mod_Ninja_Necklace_Physics
bodies {
 dyn_necklaceTag_1 [{shapeType=Capsule, length=2.5702, radius=2.8, shapeName=dyn_necklaceTag_1_capsule, center=<2.5 0 0>, rotation=<rot 90 0 0>}]
 dyn_necklaceTag_2 [{shapeType=Capsule, length=2.58, radius=2.8, shapeName=dyn_necklaceTag_2_capsule, center=<2.5 0 0>, rotation=<rot 90 0 0>}]
 dyn_necklace_1 [{shapeType=Capsule, length=5, radius=2.55, shapeName=dyn_necklace_1_capsule, center=<0.5 0.5 0>, rotation=<rot 90 0 0>}]
 dyn_necklace_2 [{shapeType=Capsule, length=5, radius=2.25, shapeName=dyn_necklace_2_capsule, center=<0.5 0.25 0>, rotation=<rot 90 0 0>}]
 dyn_necklace_3 [{shapeType=Capsule, length=5, radius=2, shapeName=dyn_necklace_3_capsule, center=<0.5 0.3 0>, rotation=<rot 90 0 0>}]
 spine_05 [{shapeType=Capsule, length=15.25, radius=10.5, shapeName=spine_05_capsule, center=<-2.6702 -1.0703 0>}]
}
constraints [
 {bone1Name=dyn_necklace_1, bone2Name=spine_05, swing1LimitDegrees=45, swing2LimitDegrees=10, twistLimitDegrees=45, twistMotion=Locked}
 {bone1Name=dyn_necklace_2, bone2Name=dyn_necklace_1, swing1LimitDegrees=45, swing2LimitDegrees=15, twistLimitDegrees=45, twistMotion=Locked}
 {bone1Name=dyn_necklace_3, bone2Name=dyn_necklace_2, swing1LimitDegrees=45, swing2LimitDegrees=20, twistLimitDegrees=45, twistMotion=Locked}
 {
  bone1Name=dyn_necklaceTag_2
  bone2Name=dyn_necklace_3
  swing1LimitDegrees=45
  swing1Motion=Locked
  swing2LimitDegrees=45
  twistLimitDegrees=45
  twistMotion=Locked
 }
 {
  bone1Name=dyn_necklaceTag_1
  bone2Name=dyn_necklace_3
  swing1LimitDegrees=45
  swing1Motion=Locked
  swing2LimitDegrees=45
  twistLimitDegrees=45
  twistMotion=Locked
 }
]
