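# Physics asset for the male medium Ninja body mesh, setting up secondary motion for its jacket and left thigh belt.
# Capsules on pelvis, spine and thighs form the body that the cloth bones hang from or can collide with.
# Seven jacket chains (left/right A, B, C and back) start at tiny master bodies locked to spine_01, with twist-locked, mostly swing-limited links.
# The left thigh belt swings and twists freely from thigh_l.
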PhysicsAsset /BRCosmetics/Characters/Player/Male/Medium/Bodies/M_MED_Mod_Ninja/Meshes/M_MED_Mod_Ninja_Physics
bodies {
 dyn_L_jacket_A_1 [{shapeType=Capsule, length=10, radius=3.5, shapeName=dyn_L_jacket_A_1_capsule, center=<4 0 0>, rotation=<rot 0 0 -40>}]
 dyn_L_jacket_A_2 [{shapeType=Capsule, length=15, radius=4, shapeName=dyn_L_jacket_A_2_capsule, center=<4 0 0>, rotation=<rot 0 0 -40>}]
 dyn_L_jacket_A_3 [{shapeType=Capsule, length=15, radius=3.6, shapeName=dyn_L_jacket_A_3_capsule, center=<4 0 0>, rotation=<rot 0 0 -40>}]
 dyn_L_jacket_A_4 [{shapeType=Capsule, length=10, radius=3.8, shapeName=dyn_L_jacket_A_4_capsule, center=<3 0 0>, rotation=<rot 0 0 -40>}]
 dyn_L_jacket_A_master [{shapeType=Capsule, length=0.505, radius=0.505, shapeName=dyn_L_jacket_A_master_capsule}]
 dyn_L_jacket_B_1 [{shapeType=Capsule, length=10, radius=3.5, shapeName=dyn_L_jacket_B_1_capsule, center=<4 0 0>}]
 dyn_L_jacket_B_2 [{shapeType=Capsule, length=10, radius=4, shapeName=dyn_L_jacket_B_2_capsule, center=<4 0 0>}]
 dyn_L_jacket_B_3 [{shapeType=Capsule, length=10, radius=4, shapeName=dyn_L_jacket_B_3_capsule, center=<4.5 0 0>, rotation=<rot 8 0 0>}]
 dyn_L_jacket_B_master [{shapeType=Capsule, length=0.505, radius=0.505, shapeName=dyn_L_jacket_B_master_capsule}]
 dyn_L_jacket_C_1 [{shapeType=Capsule, length=10, radius=3.5, shapeName=dyn_L_jacket_C_1_capsule, center=<4 0 0>}]
 dyn_L_jacket_C_2 [{shapeType=Capsule, length=10, radius=3.5, shapeName=dyn_L_jacket_C_2_capsule, center=<4 0 0>}]
 dyn_L_jacket_C_3 [{shapeType=Capsule, length=10, radius=4, shapeName=dyn_L_jacket_C_3_capsule, center=<5 0 0>}]
 dyn_L_jacket_C_master [{shapeType=Capsule, length=0.505, radius=0.505, shapeName=dyn_L_jacket_C_master_capsule}]
 dyn_L_thighBelt_front [{shapeType=Capsule, length=7, radius=0.6, shapeName=dyn_L_thighBelt_front_capsule, rotation=<rot 45 0 0>}]
 dyn_R_jacket_A_1 [{shapeType=Capsule, length=10, radius=3.5, shapeName=dyn_R_jacket_A_1_capsule, center=<-4 0 0>, rotation=<rot 0 0 -40>}]
 dyn_R_jacket_A_2 [{shapeType=Capsule, length=15, radius=4, shapeName=dyn_R_jacket_A_2_capsule, center=<-4 0 0>, rotation=<rot 0 0 -40>}]
 dyn_R_jacket_A_3 [{shapeType=Capsule, length=15, radius=3.6, shapeName=dyn_R_jacket_A_3_capsule, center=<-4 0 0>, rotation=<rot 0 0 -40>}]
 dyn_R_jacket_A_4 [{shapeType=Capsule, length=10, radius=3.8, shapeName=dyn_R_jacket_A_4_capsule, center=<-3 0 0>, rotation=<rot 0 0 -40>}]
 dyn_R_jacket_A_master [{shapeType=Capsule, length=0.505, radius=0.505, shapeName=dyn_R_jacket_A_master_capsule}]
 dyn_R_jacket_B_1 [{shapeType=Capsule, length=10, radius=3.5, shapeName=dyn_R_jacket_B_1_capsule, center=<-4 0 0>}]
 dyn_R_jacket_B_2 [{shapeType=Capsule, length=10, radius=4, shapeName=dyn_R_jacket_B_2_capsule, center=<-4 0 0>}]
 dyn_R_jacket_B_3 [{shapeType=Capsule, length=10, radius=4.5, shapeName=dyn_R_jacket_B_3_capsule, center=<-4.5 0 0>, rotation=<rot 8 0 0>}]
 dyn_R_jacket_B_master [{shapeType=Capsule, length=0.505, radius=0.505, shapeName=dyn_R_jacket_B_master_capsule}]
 dyn_R_jacket_C_1 [{shapeType=Capsule, length=10, radius=3.5, shapeName=dyn_R_jacket_C_1_capsule, center=<-4 0 0>}]
 dyn_R_jacket_C_2 [{shapeType=Capsule, length=10, radius=3.5, shapeName=dyn_R_jacket_C_2_capsule, center=<-4 0 0>}]
 dyn_R_jacket_C_3 [{shapeType=Capsule, length=10, radius=4, shapeName=dyn_R_jacket_C_3_capsule, center=<-5 0 0>}]
 dyn_R_jacket_C_master [{shapeType=Capsule, length=0.505, radius=0.505, shapeName=dyn_R_jacket_C_master_capsule}]
 dyn_backJacket_1 [{shapeType=Capsule, length=10, radius=3.7, shapeName=dyn_backJacket_1_capsule, center=<4 4.7 0>}]
 dyn_backJacket_2 [{shapeType=Capsule, length=10, radius=3.5, shapeName=dyn_backJacket_2_capsule, center=<4 3 0>}]
 dyn_backJacket_3 [{shapeType=Capsule, length=10, radius=4, shapeName=dyn_backJacket_3_capsule, center=<5 1 0>}]
 dyn_backJacket_master [{shapeType=Capsule, length=0.505, radius=0.505, shapeName=dyn_backJacket_master_capsule}]
 pelvis [{shapeType=Capsule, length=20, radius=11.5, shapeName=pelvis_capsule, center=<5.9289 0.3487 0>, rotation=<rot 90 0 0>}]
 spine_01 [{shapeType=Capsule, length=16, radius=10.5, shapeName=spine_01_capsule, center=<4 0 0>, rotation=<rot 90 0 0>}]
 spine_04 [{shapeType=Capsule, length=14, radius=11, shapeName=spine_04_capsule, center=<-5 0.5 0>, rotation=<rot 90 0 0>}]
 thigh_l [{shapeType=Capsule, length=60, radius=6, shapeName=thigh_r_capsule, center=<-12.1426 -1 0>, rotation=<rot 90 0 0>}]
 thigh_r [{shapeType=Capsule, length=40, radius=7, shapeName=thigh_r_capsule, center=<16 0 0>, rotation=<rot 90 0 0>}]
}
constraints [
 {
  bone1Name=spine_01
  bone2Name=pelvis
  swing1LimitDegrees=45
  swing1Motion=Free
  swing2LimitDegrees=45
  swing2Motion=Free
  twistLimitDegrees=45
  twistMotion=Free
 }
 {
  bone1Name=spine_04
  bone2Name=spine_01
  swing1LimitDegrees=45
  swing1Motion=Free
  swing2LimitDegrees=45
  swing2Motion=Free
  twistLimitDegrees=45
  twistMotion=Free
 }
 {
  bone1Name=dyn_backJacket_1
  bone2Name=dyn_backJacket_master
  swing1LimitDegrees=30
  swing2LimitDegrees=30
  swing2Motion=Locked
  twistLimitDegrees=45
  twistMotion=Locked
 }
 {
  bone1Name=dyn_backJacket_2
  bone2Name=dyn_backJacket_1
  swing1LimitDegrees=30
  swing2LimitDegrees=30
  twistLimitDegrees=45
  twistMotion=Locked
 }
 {
  bone1Name=dyn_backJacket_3
  bone2Name=dyn_backJacket_2
  swing1LimitDegrees=30
  swing2LimitDegrees=30
  twistLimitDegrees=45
  twistMotion=Locked
 }
 {
  bone1Name=dyn_L_jacket_C_1
  bone2Name=dyn_L_jacket_C_master
  swing1LimitDegrees=30
  swing2LimitDegrees=30
  swing2Motion=Locked
  twistLimitDegrees=45
  twistMotion=Locked
 }
 {
  bone1Name=dyn_L_jacket_C_2
  bone2Name=dyn_L_jacket_C_1
  swing1LimitDegrees=30
  swing2LimitDegrees=30
  twistLimitDegrees=45
  twistMotion=Locked
 }
 {
  bone1Name=dyn_L_jacket_C_3
  bone2Name=dyn_L_jacket_C_2
  swing1LimitDegrees=30
  swing2LimitDegrees=30
  twistLimitDegrees=45
  twistMotion=Locked
 }
 {
  bone1Name=dyn_L_jacket_B_1
  bone2Name=dyn_L_jacket_B_master
  swing1LimitDegrees=30
  swing2LimitDegrees=30
  swing2Motion=Locked
  twistLimitDegrees=45
  twistMotion=Locked
 }
 {
  bone1Name=dyn_L_jacket_B_2
  bone2Name=dyn_L_jacket_B_1
  swing1LimitDegrees=30
  swing2LimitDegrees=30
  twistLimitDegrees=45
  twistMotion=Locked
 }
 {
  bone1Name=dyn_L_jacket_B_3
  bone2Name=dyn_L_jacket_B_2
  swing1LimitDegrees=30
  swing2LimitDegrees=30
  twistLimitDegrees=45
  twistMotion=Locked
 }
 {
  bone1Name=dyn_L_jacket_A_1
  bone2Name=dyn_L_jacket_A_master
  swing1LimitDegrees=30
  swing1Motion=Free
  swing2LimitDegrees=30
  swing2Motion=Locked
  twistLimitDegrees=45
  twistMotion=Locked
 }
 {
  bone1Name=dyn_L_jacket_A_2
  bone2Name=dyn_L_jacket_A_1
  swing1LimitDegrees=30
  swing2LimitDegrees=30
  twistLimitDegrees=45
  twistMotion=Locked
 }
 {
  bone1Name=dyn_L_jacket_A_3
  bone2Name=dyn_L_jacket_A_2
  swing1LimitDegrees=30
  swing2LimitDegrees=30
  twistLimitDegrees=0
  twistMotion=Locked
 }
 {
  bone1Name=dyn_L_jacket_A_4
  bone2Name=dyn_L_jacket_A_3
  swing1LimitDegrees=30
  swing2LimitDegrees=30
  twistLimitDegrees=0
  twistMotion=Locked
 }
 {
  bone1Name=dyn_R_jacket_C_1
  bone2Name=dyn_R_jacket_C_master
  swing1LimitDegrees=30
  swing2LimitDegrees=30
  swing2Motion=Locked
  twistLimitDegrees=45
  twistMotion=Locked
 }
 {
  bone1Name=dyn_R_jacket_C_2
  bone2Name=dyn_R_jacket_C_1
  swing1LimitDegrees=30
  swing2LimitDegrees=30
  twistLimitDegrees=45
  twistMotion=Locked
 }
 {
  bone1Name=dyn_R_jacket_C_3
  bone2Name=dyn_R_jacket_C_2
  swing1LimitDegrees=30
  swing2LimitDegrees=30
  twistLimitDegrees=45
  twistMotion=Locked
 }
 {
  bone1Name=dyn_R_jacket_B_1
  bone2Name=dyn_R_jacket_B_master
  swing1LimitDegrees=30
  swing2LimitDegrees=30
  swing2Motion=Locked
  twistLimitDegrees=45
  twistMotion=Locked
 }
 {
  bone1Name=dyn_R_jacket_B_2
  bone2Name=dyn_R_jacket_B_1
  swing1LimitDegrees=30
  swing2LimitDegrees=30
  twistLimitDegrees=45
  twistMotion=Locked
 }
 {
  bone1Name=dyn_R_jacket_B_3
  bone2Name=dyn_R_jacket_B_2
  swing1LimitDegrees=30
  swing2LimitDegrees=30
  twistLimitDegrees=45
  twistMotion=Locked
 }
 {
  bone1Name=dyn_R_jacket_A_1
  bone2Name=dyn_R_jacket_A_master
  swing1LimitDegrees=30
  swing1Motion=Free
  swing2LimitDegrees=30
  swing2Motion=Locked
  twistLimitDegrees=45
  twistMotion=Locked
 }
 {
  bone1Name=dyn_R_jacket_A_2
  bone2Name=dyn_R_jacket_A_1
  swing1LimitDegrees=30
  swing2LimitDegrees=30
  twistLimitDegrees=45
  twistMotion=Locked
 }
 {
  bone1Name=dyn_R_jacket_A_3
  bone2Name=dyn_R_jacket_A_2
  swing1LimitDegrees=30
  swing2LimitDegrees=30
  twistLimitDegrees=45
  twistMotion=Locked
 }
 {
  bone1Name=dyn_R_jacket_A_4
  bone2Name=dyn_R_jacket_A_3
  swing1LimitDegrees=30
  swing2LimitDegrees=30
  twistLimitDegrees=45
  twistMotion=Locked
 }
 {
  bone1Name=thigh_l
  bone2Name=pelvis
  swing1LimitDegrees=45
  swing1Motion=Free
  swing2LimitDegrees=45
  swing2Motion=Free
  twistLimitDegrees=45
  twistMotion=Free
 }
 {
  bone1Name=thigh_r
  bone2Name=pelvis
  swing1LimitDegrees=45
  swing1Motion=Free
  swing2LimitDegrees=45
  swing2Motion=Free
  twistLimitDegrees=45
  twistMotion=Free
 }
 {
  bone1Name=dyn_L_thighBelt_front
  bone2Name=thigh_l
  swing1LimitDegrees=45
  swing1Motion=Free
  swing2LimitDegrees=45
  swing2Motion=Free
  twistLimitDegrees=45
  twistMotion=Free
 }
 {
  bone1Name=dyn_backJacket_master
  bone2Name=spine_01
  swing1LimitDegrees=45
  swing1Motion=Locked
  swing2LimitDegrees=45
  swing2Motion=Locked
  twistLimitDegrees=45
  twistMotion=Locked
 }
 {
  bone1Name=dyn_L_jacket_A_master
  bone2Name=spine_01
  swing1LimitDegrees=45
  swing1Motion=Locked
  swing2LimitDegrees=45
  swing2Motion=Locked
  twistLimitDegrees=45
  twistMotion=Locked
 }
 {
  bone1Name=dyn_L_jacket_B_master
  bone2Name=spine_01
  swing1LimitDegrees=45
  swing1Motion=Locked
  swing2LimitDegrees=45
  swing2Motion=Locked
  twistLimitDegrees=45
  twistMotion=Locked
 }
 {
  bone1Name=dyn_L_jacket_C_master
  bone2Name=spine_01
  swing1LimitDegrees=45
  swing1Motion=Locked
  swing2LimitDegrees=45
  swing2Motion=Locked
  twistLimitDegrees=45
  twistMotion=Locked
 }
 {
  bone1Name=dyn_R_jacket_A_master
  bone2Name=spine_01
  swing1LimitDegrees=45
  swing1Motion=Locked
  swing2LimitDegrees=45
  swing2Motion=Locked
  twistLimitDegrees=45
  twistMotion=Locked
 }
 {
  bone1Name=dyn_R_jacket_B_master
  bone2Name=spine_01
  swing1LimitDegrees=45
  swing1Motion=Locked
  swing2LimitDegrees=45
  swing2Motion=Locked
  twistLimitDegrees=45
  twistMotion=Locked
 }
 {
  bone1Name=dyn_R_jacket_C_master
  bone2Name=spine_01
  swing1LimitDegrees=45
  swing1Motion=Locked
  swing2LimitDegrees=45
  swing2Motion=Locked
  twistLimitDegrees=45
  twistMotion=Locked
 }
]
