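# Post-copy AnimBP for the male medium Ninja body cosmetic, adding secondary motion and corrective posing to jacket, straps, buckles, pouches and laces.
# The pose copied from the attached parent mesh passes through additive bone offsets, weighted constraints, two RigidBody simulations and chained AnimDynamics springs.
# EventGraph derives offset and constraint alphas from left-thigh yaw, crouch, skydive/parachute state and a front-end idle montage.
# The first two lower back-strap AnimDynamics nodes never reach the output.
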AnimBlueprint /BRCosmetics/Characters/Player/Male/Medium/Bodies/M_MED_Mod_Ninja/Meshes/M_MED_Mod_Ninja_AnimBP
target_skeleton=&/M_MED_Mod_Ninja_Skeleton
graphs [
 AnimGraph {nodes [
   n1 = Anim Dynamics - Bone: dyn_L_backStrap_bottom_1(Alpha=1, ExternalForce="0,0,0") {
    angularDampingOverride=1
    angularSpringConstant=20000
    bAngularSpring=true
    bOverrideAngularDamping=true
    bUseGravityOverride=true
    gravityScale=0.35
    simulationSpace=RootRelative
    alphaBoolBlend {blendOption=Linear}
    boundBone {boneName=dyn_L_backStrap_bottom_1}
    componentLinearAccScale=<1 1 1>
    componentLinearVelScale=<1 1 1>
    gravityOverride {z=-2500}
    physicsBodyDefinitions [{
      boundBone {boneName=dyn_L_backStrap_bottom_1}
      boxExtents=<7.5 7.5 7.5>
      constraintSetup {angularLimitsMax=<0 30 0>, angularLimitsMin=<0 -30 -20>, angularTarget=<0 0 0>}
      localJointOffset=<1 0 0>
     }]
   }
   Anim Dynamics - Bone: dyn_R_backStrap_bottom_1(ComponentPose=@n1, Alpha=1, ExternalForce="0,0,0") {
    angularDampingOverride=1
    angularSpringConstant=20000
    bAngularSpring=true
    bOverrideAngularDamping=true
    bUseGravityOverride=true
    gravityScale=0.35
    simulationSpace=RootRelative
    alphaBoolBlend {blendOption=Linear}
    boundBone {boneName=dyn_R_backStrap_bottom_1}
    componentLinearAccScale=<1 1 1>
    componentLinearVelScale=<1 1 1>
    gravityOverride {z=-2500}
    physicsBodyDefinitions [{
      boundBone {boneName=dyn_R_backStrap_bottom_1}
      boxExtents=<8 8 8>
      constraintSetup {angularLimitsMax=<0 30 0>, angularLimitsMin=<0 -30 -20>, angularTarget=<0 0 0>}
      localJointOffset=<-1 0 0>
     }]
   }
   n2 = Get L_thighRotationAlpha() -> (L_thighRotationAlpha:double)
   n3 = "Transform (Modify) Bone - Bone: dyn_L_jacket_A_master"(
    ComponentPose=Local To Component(LocalPose=Copy Pose From Mesh() {bUseAttachedParent=true})
    Alpha=@n2
    Translation="0,0,0"
    Rotation="0.000,-44,-70"
    Scale="1,1,1"
   ) {
    rotationMode=BMM_Additive
    translationMode=BMM_Additive
    alphaBoolBlend {blendOption=Linear}
    alphaScaleBiasClamp {interpSpeedDecreasing=20}
    boneToModify {boneName=dyn_L_jacket_A_master}
    rotation {roll=-70, yaw=-44}
   }
   n4 = "Transform (Modify) Bone - Bone: dyn_backJacket_master"(ComponentPose=@n3, Alpha=Get isCrouchMoving() -> (isCrouchMoving:double), Translation="0,0,0", Rotation="0,80,0", Scale="1,1,1") {
    rotationMode=BMM_Additive
    rotationSpace=BCS_BoneSpace
    translationMode=BMM_Additive
    alphaBoolBlend {blendOption=Linear}
    alphaScaleBiasClamp {bInterpResult=true, interpSpeedDecreasing=20, interpSpeedIncreasing=20}
    boneToModify {boneName=dyn_backJacket_master}
    rotation {yaw=80}
   }
   n5 = "Transform (Modify) Bone - Bone: dyn_L_jacket_B_master"(
    ComponentPose=@n4
    Alpha=Get Is Crouching Not Moving() -> (Is Crouching Not Moving:double)
    Translation="-8.000,0.000,-8.085"
    Rotation="54,-59,-124.791"
    Scale="1,1,1"
   ) {
    rotationMode=BMM_Additive
    translationMode=BMM_Additive
    alphaBoolBlend {blendOption=Linear}
    alphaScaleBiasClamp {bInterpResult=true}
    boneToModify {boneName=dyn_L_jacket_B_master}
    rotation=<rot 54 -59 -124.791>
    translation {x=-8, z=-8.085}
   }
   n6 = "Transform (Modify) Bone - Bone: dyn_L_frontBuckle_bottom"(ComponentPose=@n5, Alpha=@n2, Translation="0.000,0.000,-2.2", Rotation="17,15.6,-5", Scale="1,1,1") {
    rotationMode=BMM_Additive
    rotationSpace=BCS_BoneSpace
    translationMode=BMM_Additive
    alphaBoolBlend {blendOption=Linear}
    boneToModify {boneName=dyn_L_frontBuckle_bottom}
    rotation=<rot 17 15.6 -5>
    translation {z=-2.2}
   }
   n7 = "Constraint - dyn_pouchVolume (PP)"!(ComponentPose=@n6, Alpha=Get pocketAlpha() -> (pocketAlpha:double), ConstraintWeights_0=0.5, ConstraintWeights_1:float=0.5) -> (Pose) {
    alphaBoolBlend {blendOption=Linear}
    boneToModify {boneName=dyn_pouchVolume}
    constraintSetup [{targetBone {boneName=dyn_L_jacket_A_2}}, {targetBone {boneName=dyn_L_jacket_B_2}}]
    constraintWeights [1,1]
   }
   n8 = RigidBody(ComponentPose=@n7) {
    bClampLinearTranslationLimitToRefPose=true
    bOverrideWorldGravity=true
    overridePhysicsAsset=&/M_MED_Mod_Ninja_Physics
    simulationSpace=BaseBoneSpace
    alphaBoolBlend {blendOption=Linear}
    baseBoneRef {boneName=pelvis}
    componentAppliedLinearAccClamp=<3000 3000 3000>
    componentLinearAccScale=<0.3 0.3 0.3>
    componentLinearVelScale=<0.3 0.3 0.3>
    overrideWorldGravity {z=-980}
   }
   n9 = Get skydivingFinal() -> (skydivingFinal:double)
   n10 = "Transform (Modify) Bone - Bone: dyn_backJacket_master"!(ComponentPose=@n8, Alpha=@n9, Rotation="0,0,20", Scale="1,1,1") -> (Pose) {
    rotationMode=BMM_Additive
    alphaBoolBlend {blendOption=Linear}
    alphaScaleBiasClamp {bInterpResult=true}
    boneToModify {boneName=dyn_backJacket_master}
    rotation {roll=20}
   }
   n11 = "Transform (Modify) Bone - Bone: dyn_R_jacket_C_master"!(ComponentPose=@n10, Alpha=@n9, Rotation="0,-10,40", Scale="1,1,1") -> (Pose) {
    rotationMode=BMM_Additive
    alphaBoolBlend {blendOption=Linear}
    alphaScaleBiasClamp {bInterpResult=true}
    boneToModify {boneName=dyn_R_jacket_C_master}
    rotation {roll=40, yaw=-10}
   }
   n12 = "Transform (Modify) Bone - Bone: dyn_L_jacket_C_master"!(ComponentPose=@n11, Alpha=@n9, Rotation="0,10,40", Scale="1,1,1") -> (Pose) {
    rotationMode=BMM_Additive
    alphaBoolBlend {blendOption=Linear}
    alphaScaleBiasClamp {bInterpResult=true}
    boneToModify {boneName=dyn_L_jacket_C_master}
    rotation {roll=40, yaw=10}
   }
   n13 = RigidBody(ComponentPose=@n12) {
    bClampLinearTranslationLimitToRefPose=true
    bOverrideWorldGravity=true
    overridePhysicsAsset=&/M_MED_Mod_Ninja_Necklace_Physics
    simulationSpace=BaseBoneSpace
    alphaBoolBlend {blendOption=Linear}
    baseBoneRef {boneName=pelvis}
    componentAppliedLinearAccClamp=<3000 3000 3000>
    componentLinearAccScale=<0.3 0.3 0.3>
    componentLinearVelScale=<0.4 0.4 0.4>
    overrideWorldGravity {z=-980}
   }
   n14 = Get isCrouchMoving()
   n15 = "Transform (Modify) Bone - Bone: dyn_backJacket_1"(ComponentPose=@n13, Alpha=@n14, Translation="0,1,-1", Rotation="0,-10,0", Scale="1,1,1") {
    rotationMode=BMM_Additive
    rotationSpace=BCS_BoneSpace
    translationMode=BMM_Additive
    alphaBoolBlend {blendOption=Linear}
    alphaScaleBiasClamp {bInterpResult=true, interpSpeedIncreasing=5}
    boneToModify {boneName=dyn_backJacket_1}
    rotation {yaw=-10}
    translation {y=1, z=-1}
   }
   n16 = "Transform (Modify) Bone - Bone: dyn_backJacket_2"(ComponentPose=@n15, Alpha=@n14, Translation="0,0,-2", Rotation="0,-10,0", Scale="1,1,1") {
    rotationMode=BMM_Additive
    rotationSpace=BCS_BoneSpace
    translationMode=BMM_Additive
    alphaBoolBlend {blendOption=Linear}
    alphaScaleBiasClamp {bInterpResult=true, interpSpeedIncreasing=5}
    boneToModify {boneName=dyn_backJacket_2}
    rotation {yaw=-10}
    translation {z=-2}
   }
   n17 = "Constraint - dyn_L_backStrap_top (PPPPPP)"!(
    ComponentPose=@n16
    Alpha=1
    ConstraintWeights_0=0.0061
    ConstraintWeights_1:float=0.028
    ConstraintWeights_2:float=0.001
    ConstraintWeights_3:float=0.0004
    ConstraintWeights_4:float=0.9609
    ConstraintWeights_5:float=0.0036
   ) -> (Pose) {
    alphaBoolBlend {blendOption=Linear}
    boneToModify {boneName=dyn_L_backStrap_top}
    constraintSetup [
     {targetBone {boneName=spine_04}}
     {targetBone {boneName=clavicle_l}}
     {targetBone {boneName=dyn_backJacket_2}}
     {targetBone {boneName=upperarm_twist_01_l}}
     {targetBone {boneName=spine_05}}
     {targetBone {boneName=dyn_L_jacket_C_1}}
    ]
    constraintWeights [1,1,1,1,1,1]
   }
   n18 = "Constraint - dyn_R_backStrap_top (PPPPPPP)"!(
    ComponentPose=@n17
    Alpha=1
    ConstraintWeights_0=0.0008
    ConstraintWeights_1:float=0.0061
    ConstraintWeights_2:float=0.0004
    ConstraintWeights_3:float=0.0035
    ConstraintWeights_4:float=0.0279
    ConstraintWeights_5:float=0.001
    ConstraintWeights_6:float=0.9602
   ) -> (Pose) {
    alphaBoolBlend {blendOption=Linear}
    boneToModify {boneName=dyn_R_backStrap_top}
    constraintSetup [
     {targetBone {boneName=dyn_backJacket_1}}
     {targetBone {boneName=spine_04}}
     {targetBone {boneName=upperarm_twist_01_r}}
     {targetBone {boneName=dyn_R_jacket_C_1}}
     {targetBone {boneName=clavicle_r}}
     {targetBone {boneName=dyn_backJacket_2}}
     {targetBone {boneName=spine_05}}
    ]
    constraintWeights [1,1,1,1,1,1,1]
   }
   n19 = Anim Dynamics - Bone: dyn_L_frontBuckle_top(ComponentPose=@n18, Alpha=1, ExternalForce="0,0,0") {
    angularDampingOverride=1
    angularSpringConstant=20000
    bAngularSpring=true
    bOverrideAngularDamping=true
    bUseGravityOverride=true
    gravityScale=0.35
    simulationSpace=RootRelative
    alphaBoolBlend {blendOption=Linear}
    boundBone {boneName=dyn_L_frontBuckle_top}
    componentLinearAccScale=<1 1 1>
    componentLinearVelScale=<1 1 1>
    gravityOverride {z=-980}
    physicsBodyDefinitions [{
      boundBone {boneName=dyn_L_frontBuckle_top}
      boxExtents=<7 7 7>
      constraintSetup {angularLimitsMax=<0 -6 20>, angularLimitsMin=<0 -20 -5>, angularTarget=<0 0 0>}
      localJointOffset=<2 0 0>
     }]
   }
   n20 = Anim Dynamics - Bone: dyn_R_frontBuckle_top(ComponentPose=@n19, Alpha=1, ExternalForce="0,0,0") {
    angularDampingOverride=1
    angularSpringConstant=20000
    bAngularSpring=true
    bOverrideAngularDamping=true
    bUseGravityOverride=true
    gravityScale=0.35
    simulationSpace=RootRelative
    alphaBoolBlend {blendOption=Linear}
    boundBone {boneName=dyn_R_frontBuckle_top}
    componentLinearAccScale=<1 1 1>
    componentLinearVelScale=<1 1 1>
    gravityOverride {z=-980}
    physicsBodyDefinitions [{
      boundBone {boneName=dyn_R_frontBuckle_top}
      boxExtents=<7.5 7.5 7.5>
      constraintSetup {angularLimitsMax=<0 -6 20>, angularLimitsMin=<0 -20 -5>, angularTarget=<0 0 0>}
      localJointOffset=<-2 0 0>
     }]
   }
   n21 = Anim Dynamics - Bone: dyn_L_frontBuckle_bottom(ComponentPose=@n20, Alpha=1, ExternalForce="0,0,0") {
    angularDampingOverride=1
    angularSpringConstant=20000
    bAngularSpring=true
    bOverrideAngularDamping=true
    bUseGravityOverride=true
    gravityScale=0.35
    simulationSpace=RootRelative
    alphaBoolBlend {blendOption=Linear}
    boundBone {boneName=dyn_L_frontBuckle_bottom}
    componentLinearAccScale=<1 1 1>
    componentLinearVelScale=<1 1 1>
    gravityOverride {z=-980}
    physicsBodyDefinitions [{
      boundBone {boneName=dyn_L_frontBuckle_bottom}
      boxExtents=<7.5 7.5 7.5>
      constraintSetup {angularLimitsMax=<0 -6 20>, angularLimitsMin=<0 -20 -5>, angularTarget=<0 0 0>}
      localJointOffset=<2 0 0>
     }]
    relativeSpaceBone {boneName=dyn_pouchVolume}
   }
   n22 = Anim Dynamics - Bone: dyn_R_frontBuckle_bottom(ComponentPose=@n21, Alpha=1, ExternalForce="0,0,0") {
    angularDampingOverride=1
    angularSpringConstant=20000
    bAngularSpring=true
    bOverrideAngularDamping=true
    bUseGravityOverride=true
    gravityScale=0.35
    simulationSpace=RootRelative
    alphaBoolBlend {blendOption=Linear}
    boundBone {boneName=dyn_R_frontBuckle_bottom}
    componentLinearAccScale=<1 1 1>
    componentLinearVelScale=<1 1 1>
    gravityOverride {z=-980}
    physicsBodyDefinitions [{
      boundBone {boneName=dyn_R_frontBuckle_bottom}
      boxExtents=<7 7 7>
      constraintSetup {angularLimitsMax=<0 -6 20>, angularLimitsMin=<0 -20 -5>, angularTarget=<0 0 0>}
      localJointOffset=<-2 0 0>
     }]
   }
   n23 = Anim Dynamics - Bone: dyn_L_backStrap_top(ComponentPose=@n22, Alpha=1, ExternalForce="0,0,0") {
    angularDampingOverride=1
    angularSpringConstant=20000
    bAngularSpring=true
    bOverrideAngularDamping=true
    bUseGravityOverride=true
    gravityScale=0.35
    simulationSpace=RootRelative
    alphaBoolBlend {blendOption=Linear}
    boundBone {boneName=dyn_L_backStrap_top}
    componentLinearAccScale=<1 1 1>
    componentLinearVelScale=<1 1 1>
    gravityOverride {z=-980}
    physicsBodyDefinitions [{
      boundBone {boneName=dyn_L_backStrap_top}
      boxExtents=<30 30 30>
      constraintSetup {angularLimitsMax=<0 10 10>, angularLimitsMin=<0 -30 -10>, angularTarget=<0 0 0>}
      localJointOffset=<2 0 0>
     }]
   }
   n24 = Anim Dynamics - Bone: dyn_R_backStrap_top(ComponentPose=@n23, Alpha=1, ExternalForce="0,0,0") {
    angularDampingOverride=1
    angularSpringConstant=20000
    bAngularSpring=true
    bOverrideAngularDamping=true
    bUseGravityOverride=true
    gravityScale=0.35
    simulationSpace=RootRelative
    alphaBoolBlend {blendOption=Linear}
    boundBone {boneName=dyn_R_backStrap_top}
    componentLinearAccScale=<1 1 1>
    componentLinearVelScale=<1 1 1>
    gravityOverride {z=-980}
    physicsBodyDefinitions [{
      boundBone {boneName=dyn_R_backStrap_top}
      boxExtents=<30 30 30>
      constraintSetup {angularLimitsMax=<0 10 10>, angularLimitsMin=<0 -30 -10>, angularTarget=<0 0 0>}
      localJointOffset=<-2 0 0>
     }]
   }
   n25 = Anim Dynamics - Bone: dyn_L_bootLace_1(ComponentPose=@n24, Alpha=1, ExternalForce="0,0,0") {
    angularDampingOverride=0.999
    angularSpringConstant=20000
    bAngularSpring=true
    bOverrideAngularDamping=true
    bUseGravityOverride=true
    gravityScale=0.35
    simulationSpace=RootRelative
    alphaBoolBlend {blendOption=Linear}
    boundBone {boneName=dyn_L_bootLace_1}
    componentLinearAccScale=<1 1 1>
    componentLinearVelScale=<1 1 1>
    gravityOverride {z=-980}
    physicsBodyDefinitions [{
      boundBone {boneName=dyn_L_bootLace_1}
      boxExtents=<10 10 10>
      constraintSetup {angularLimitsMax=<0 20 20>, angularLimitsMin=<0 -30 -20>, angularTarget=<0 0 0>}
      localJointOffset=<1 0 0>
     }]
   }
   n26 = Anim Dynamics - Bone: dyn_L_bootLace_2(ComponentPose=@n25, Alpha=1, ExternalForce="0,0,0") {
    angularDampingOverride=0.999
    angularSpringConstant=20000
    bAngularSpring=true
    bOverrideAngularDamping=true
    bUseGravityOverride=true
    gravityScale=0.35
    simulationSpace=RootRelative
    alphaBoolBlend {blendOption=Linear}
    boundBone {boneName=dyn_L_bootLace_2}
    componentLinearAccScale=<1 1 1>
    componentLinearVelScale=<1 1 1>
    gravityOverride {z=-980}
    physicsBodyDefinitions [{
      boundBone {boneName=dyn_L_bootLace_2}
      boxExtents=<10 10 10>
      constraintSetup {angularLimitsMax=<0 10 20>, angularLimitsMin=<0 -30 -20>, angularTarget=<0 0 0>}
      localJointOffset=<1 0 0>
     }]
   }
   n27 = Anim Dynamics - Bone: dyn_R_bootLace(ComponentPose=@n26, Alpha=1, ExternalForce="0,0,0") {
    angularDampingOverride=0.999
    angularSpringConstant=20000
    bAngularSpring=true
    bOverrideAngularDamping=true
    bUseGravityOverride=true
    gravityScale=0.35
    simulationSpace=RootRelative
    alphaBoolBlend {blendOption=Linear}
    boundBone {boneName=dyn_R_bootLace}
    componentLinearAccScale=<1 1 1>
    componentLinearVelScale=<1 1 1>
    gravityOverride {z=-980}
    physicsBodyDefinitions [{
      boundBone {boneName=dyn_R_bootLace}
      boxExtents=<10 10 10>
      constraintSetup {angularLimitsMax=<0 15 20>, angularLimitsMin=<0 -30 -20>, angularTarget=<0 0 0>}
      localJointOffset=<1 0 0>
     }]
   }
   n28 = Anim Dynamics - Bone: dyn_L_bagStrap_A_1(ComponentPose=@n27, Alpha=1, ExternalForce="0,0,0") {
    angularDampingOverride=0.9999
    angularSpringConstant=20000
    bAngularSpring=true
    bOverrideAngularDamping=true
    bUseGravityOverride=true
    gravityScale=0.35
    simulationSpace=RootRelative
    alphaBoolBlend {blendOption=Linear}
    boundBone {boneName=dyn_L_bagStrap_A_1}
    componentLinearAccScale=<1 1 1>
    componentLinearVelScale=<1 1 1>
    gravityOverride {z=-980}
    physicsBodyDefinitions [{
      boundBone {boneName=dyn_L_bagStrap_A_1}
      boxExtents=<6 6 6>
      constraintSetup {angularLimitsMax=<0 30 30>, angularLimitsMin=<0 -30 -30>, angularTarget=<0 0 0>}
      localJointOffset=<1 0 0>
     }]
   }
   n29 = Anim Dynamics - Bone: dyn_L_bagStrap_A_2(ComponentPose=@n28, Alpha=1, ExternalForce="0,0,0") {
    angularDampingOverride=0.9999
    angularSpringConstant=20000
    bAngularSpring=true
    bOverrideAngularDamping=true
    bUseGravityOverride=true
    gravityScale=0.35
    simulationSpace=RootRelative
    alphaBoolBlend {blendOption=Linear}
    boundBone {boneName=dyn_L_bagStrap_A_2}
    componentLinearAccScale=<1 1 1>
    componentLinearVelScale=<1 1 1>
    gravityOverride {z=-980}
    physicsBodyDefinitions [{
      boundBone {boneName=dyn_L_bagStrap_A_2}
      boxExtents=<6 6 6>
      constraintSetup {angularLimitsMax=<0 30 30>, angularLimitsMin=<0 -30 -30>, angularTarget=<0 0 0>}
      localJointOffset=<1 0 0>
     }]
   }
   n30 = Anim Dynamics - Bone: dyn_R_bagStrap_A_1(ComponentPose=@n29, Alpha=1, ExternalForce="0,0,0") {
    angularDampingOverride=0.9999
    angularSpringConstant=20000
    bAngularSpring=true
    bOverrideAngularDamping=true
    bUseGravityOverride=true
    gravityScale=0.35
    simulationSpace=RootRelative
    alphaBoolBlend {blendOption=Linear}
    boundBone {boneName=dyn_R_bagStrap_A_1}
    componentLinearAccScale=<1 1 1>
    componentLinearVelScale=<1 1 1>
    gravityOverride {z=-980}
    physicsBodyDefinitions [{
      boundBone {boneName=dyn_R_bagStrap_A_1}
      boxExtents=<8 8 8>
      constraintSetup {angularLimitsMax=<0 30 30>, angularLimitsMin=<0 -30 -30>, angularTarget=<0 0 0>}
      localJointOffset=<-1 0 0>
     }]
   }
   n31 = Anim Dynamics - Bone: dyn_R_bagStrap_A_2(ComponentPose=@n30, Alpha=1, ExternalForce="0,0,0") {
    angularDampingOverride=0.9999
    angularSpringConstant=20000
    bAngularSpring=true
    bOverrideAngularDamping=true
    bUseGravityOverride=true
    gravityScale=0.35
    simulationSpace=RootRelative
    alphaBoolBlend {blendOption=Linear}
    boundBone {boneName=dyn_R_bagStrap_A_2}
    componentLinearAccScale=<1 1 1>
    componentLinearVelScale=<1 1 1>
    gravityOverride {z=-980}
    physicsBodyDefinitions [{
      boundBone {boneName=dyn_R_bagStrap_A_2}
      boxExtents=<8 8 8>
      constraintSetup {angularLimitsMax=<0 30 30>, angularLimitsMin=<0 -30 -30>, angularTarget=<0 0 0>}
      localJointOffset=<-1 0 0>
     }]
   }
   n32 = Anim Dynamics - Bone: dyn_L_bagStrap_B_1(ComponentPose=@n31, Alpha=1, ExternalForce="0,0,0") {
    angularDampingOverride=0.9999
    angularSpringConstant=20000
    bAngularSpring=true
    bOverrideAngularDamping=true
    bUseGravityOverride=true
    gravityScale=0.35
    simulationSpace=RootRelative
    alphaBoolBlend {blendOption=Linear}
    boundBone {boneName=dyn_L_bagStrap_B_1}
    componentLinearAccScale=<1 1 1>
    componentLinearVelScale=<1 1 1>
    gravityOverride {z=-980}
    physicsBodyDefinitions [{
      boundBone {boneName=dyn_L_bagStrap_B_1}
      boxExtents=<8 8 8>
      constraintSetup {angularLimitsMax=<0 30 30>, angularLimitsMin=<0 -30 -30>, angularTarget=<0 0 0>}
      localJointOffset=<1 0 0>
     }]
   }
   n33 = Anim Dynamics - Bone: dyn_L_bagStrap_B_2(ComponentPose=@n32, Alpha=1, ExternalForce="0,0,0") {
    angularDampingOverride=0.9999
    angularSpringConstant=20000
    bAngularSpring=true
    bOverrideAngularDamping=true
    bUseGravityOverride=true
    gravityScale=0.35
    simulationSpace=RootRelative
    alphaBoolBlend {blendOption=Linear}
    boundBone {boneName=dyn_L_bagStrap_B_2}
    componentLinearAccScale=<1 1 1>
    componentLinearVelScale=<1 1 1>
    gravityOverride {z=-980}
    physicsBodyDefinitions [{
      boundBone {boneName=dyn_L_bagStrap_B_2}
      boxExtents=<8 8 8>
      constraintSetup {angularLimitsMax=<0 30 30>, angularLimitsMin=<0 -30 -30>, angularTarget=<0 0 0>}
      localJointOffset=<1 0 0>
     }]
   }
   n34 = Anim Dynamics - Bone: dyn_R_bagStrap_B_1(ComponentPose=@n33, Alpha=1, ExternalForce="0,0,0") {
    angularDampingOverride=0.9999
    angularSpringConstant=20000
    bAngularSpring=true
    bOverrideAngularDamping=true
    bUseGravityOverride=true
    gravityScale=0.35
    simulationSpace=RootRelative
    alphaBoolBlend {blendOption=Linear}
    boundBone {boneName=dyn_R_bagStrap_B_1}
    componentLinearAccScale=<1 1 1>
    componentLinearVelScale=<1 1 1>
    gravityOverride {z=-980}
    physicsBodyDefinitions [{
      boundBone {boneName=dyn_R_bagStrap_B_1}
      boxExtents=<6 6 6>
      constraintSetup {angularLimitsMax=<0 30 30>, angularLimitsMin=<0 -30 -30>, angularTarget=<0 0 0>}
      localJointOffset=<-1 0 0>
     }]
   }
   n35 = Anim Dynamics - Bone: dyn_R_bagStrap_B_2(ComponentPose=@n34, Alpha=1, ExternalForce="0,0,0") {
    angularDampingOverride=0.9999
    angularSpringConstant=20000
    bAngularSpring=true
    bOverrideAngularDamping=true
    bUseGravityOverride=true
    gravityScale=0.35
    simulationSpace=RootRelative
    alphaBoolBlend {blendOption=Linear}
    boundBone {boneName=dyn_R_bagStrap_B_2}
    componentLinearAccScale=<1 1 1>
    componentLinearVelScale=<1 1 1>
    gravityOverride {z=-980}
    physicsBodyDefinitions [{
      boundBone {boneName=dyn_R_bagStrap_B_2}
      boxExtents=<6 6 6>
      constraintSetup {angularLimitsMax=<0 30 30>, angularLimitsMin=<0 -30 -30>, angularTarget=<0 0 0>}
      localJointOffset=<-1 0 0>
     }]
   }
   n36 = Anim Dynamics - Bone: dyn_L_backStrap_bottom_2(ComponentPose=@n35, Alpha=1, ExternalForce="0,0,0") {
    angularDampingOverride=1
    angularSpringConstant=20000
    bAngularSpring=true
    bOverrideAngularDamping=true
    bUseGravityOverride=true
    gravityScale=0.35
    simulationSpace=RootRelative
    alphaBoolBlend {blendOption=Linear}
    boundBone {boneName=dyn_L_backStrap_bottom_2}
    componentLinearAccScale=<1 1 1>
    componentLinearVelScale=<1 1 1>
    gravityOverride {z=-2500}
    physicsBodyDefinitions [{
      boundBone {boneName=dyn_L_backStrap_bottom_2}
      boxExtents=<7.5 7.5 7.5>
      constraintSetup {angularLimitsMax=<0 30 0>, angularLimitsMin=<0 -30 -20>, angularTarget=<0 0 0>}
      localJointOffset=<1 0 0>
     }]
   }
   n37 = Anim Dynamics - Bone: dyn_R_backStrap_bottom_2(ComponentPose=@n36, Alpha=1, ExternalForce="0,0,0") {
    angularDampingOverride=1
    angularSpringConstant=20000
    bAngularSpring=true
    bOverrideAngularDamping=true
    bUseGravityOverride=true
    gravityScale=0.35
    simulationSpace=RootRelative
    alphaBoolBlend {blendOption=Linear}
    boundBone {boneName=dyn_R_backStrap_bottom_2}
    componentLinearAccScale=<1 1 1>
    componentLinearVelScale=<1 1 1>
    gravityOverride {z=-2500}
    physicsBodyDefinitions [{
      boundBone {boneName=dyn_R_backStrap_bottom_2}
      boxExtents=<8 8 8>
      constraintSetup {angularLimitsMax=<0 30 0>, angularLimitsMin=<0 -30 -20>, angularTarget=<0 0 0>}
      localJointOffset=<-1 0 0>
     }]
   }
   n38 = "Constraint - dyn_L_armPouch (PPPPPP)"!(
    ComponentPose=@n37
    Alpha=1
    ConstraintWeights_0=0.147
    ConstraintWeights_1:float=0.0012
    ConstraintWeights_2:float=0
    ConstraintWeights_3:float=0.0019
    ConstraintWeights_4:float=0.1741
    ConstraintWeights_5:float=0.6758
   ) -> (Pose) {
    alphaBoolBlend {blendOption=Linear}
    boneToModify {boneName=dyn_L_armPouch}
    constraintSetup [
     {targetBone {boneName=upperarm_l}}
     {targetBone {boneName=lowerarm_l}}
     {targetBone {boneName=lowerarm_twist_01_l}}
     {targetBone {boneName=lowerarm_twist_02_l}}
     {targetBone {boneName=upperarm_twist_01_l}}
     {targetBone {boneName=upperarm_twist_02_l}}
    ]
    constraintWeights [1,1,1,1,1,1]
   }
   n39 = "Constraint - dyn_L_thighBelt_back (PP)"!(ComponentPose=@n38, Alpha=1, ConstraintWeights_0=0.8, ConstraintWeights_1:float=0.2) -> (Pose) {
    alphaBoolBlend {blendOption=Linear}
    boneToModify {boneName=dyn_L_thighBelt_back}
    constraintSetup [{targetBone {boneName=pelvis}}, {targetBone {boneName=thigh_l}}]
    constraintWeights [1,1]
   }
   n40 = "Constraint - dyn_R_thighBelt_back_2 (PP)"!(ComponentPose=@n39, Alpha=1, ConstraintWeights_0=0.89, ConstraintWeights_1:float=0.11) -> (Pose) {
    alphaBoolBlend {blendOption=Linear}
    boneToModify {boneName=dyn_R_thighBelt_back_2}
    constraintSetup [{targetBone {boneName=pelvis}}, {targetBone {boneName=thigh_r}}]
    constraintWeights [1,1]
   }
   n41 = "Constraint - dyn_R_thighBelt_front_2 (PPPP)"!(
    ComponentPose=@n40
    Alpha=1
    ConstraintWeights_0=0.1801
    ConstraintWeights_1:float=0.0144
    ConstraintWeights_2:float=0.7709
    ConstraintWeights_3:float=0.0345
   ) -> (Pose) {
    alphaBoolBlend {blendOption=Linear}
    boneToModify {boneName=dyn_R_thighBelt_front_2}
    constraintSetup [{targetBone {boneName=pelvis}}, {targetBone {boneName=thigh_l}}, {targetBone {boneName=thigh_r}}, {targetBone {boneName=thigh_twist_01_r}}]
    constraintWeights [1,1,1,1]
   }
   n42 = Get isIdleEmotePlaying() -> (isIdleEmotePlaying:double)
   n43 = "Transform (Modify) Bone - Bone: dyn_R_legPouch"(ComponentPose=@n41, Alpha=@n42, Translation="0,-3,0", Rotation="-3,0,3.5", Scale="1,1,1") {
    rotationMode=BMM_Additive
    translationMode=BMM_Additive
    translationSpace=BCS_BoneSpace
    alphaBoolBlend {blendOption=Linear}
    boneToModify {boneName=dyn_R_legPouch}
    rotation {pitch=-3, roll=3.5}
    translation {y=-3}
   }
   n44 = "Transform (Modify) Bone - Bone: dyn_R_thighBelt_front_2"(ComponentPose=@n43, Alpha=@n42, Translation="0,-1,0", Rotation="2,4,1", Scale="1,1,1") {
    rotationMode=BMM_Additive
    translationMode=BMM_Additive
    translationSpace=BCS_BoneSpace
    alphaBoolBlend {blendOption=Linear}
    boneToModify {boneName=dyn_R_thighBelt_front_2}
    rotation=<rot 2 4 1>
    translation {y=-1}
   }
   n45 = "Transform (Modify) Bone - Bone: dyn_R_thighBelt_back_2"(ComponentPose=@n44, Alpha=@n42, Translation="-1,-1.2,1.2", Rotation="0,0,0", Scale="1,1,1") {
    rotationMode=BMM_Additive
    translationMode=BMM_Additive
    translationSpace=BCS_BoneSpace
    alphaBoolBlend {blendOption=Linear}
    boneToModify {boneName=dyn_R_thighBelt_back_2}
    translation=<-1 -1.2 1.2>
   }
   n46 = "Transform (Modify) Bone - Bone: dyn_L_thighBelt_front"(ComponentPose=@n45, Alpha=1, Translation="-2,0,0", Rotation="0,0,0", Scale="1,1,1") {
    rotationMode=BMM_Additive
    translationMode=BMM_Additive
    translationSpace=BCS_BoneSpace
    alphaBoolBlend {blendOption=Linear}
    boneToModify {boneName=dyn_L_thighBelt_front}
    translation {x=-2}
   }
   n47 = "Transform (Modify) Bone - Bone: dyn_L_thighBelt_back"(ComponentPose=@n46, Alpha=1, Translation="-4.064,0.951,0.206", Rotation="6.932,-0.000,15.000", Scale="1,1,1") {
    rotationMode=BMM_Additive
    translationMode=BMM_Additive
    translationSpace=BCS_BoneSpace
    alphaBoolBlend {blendOption=Linear}
    boneToModify {boneName=dyn_L_thighBelt_back}
    rotation=<rot 6.932 0 15>
    translation=<-4.064 0.951 0.206>
   }
   n48 = "Transform (Modify) Bone - Bone: dyn_L_legPouch"(ComponentPose=@n47, Alpha=1, Translation="0,0,0", Rotation="-0.244,6.235,2.165", Scale="1,1,1") {
    rotationMode=BMM_Additive
    translationMode=BMM_Additive
    translationSpace=BCS_BoneSpace
    alphaBoolBlend {blendOption=Linear}
    boneToModify {boneName=dyn_L_legPouch}
    rotation=<rot -0.244 6.235 2.165>
   }
   Output Pose(Result=Component To Local(ComponentPose=@n48)) {layerGroup=None}
  ]}
 EventGraph {nodes [
   Event BlueprintInitializeAnimation()
   > n1 = Set Skel_Mesh(Skel_Mesh=GetOwnerSkeletalMeshComponent())
   > Is Valid(InputObject=@n1)
   > Set Anim_Inst(Anim_Inst:Anim Instance Object Reference=GetAnimInstance(self=@n1)) -> (Output_Get:Anim Instance Object Reference)
   Event BlueprintUpdateAnimation()
   > n2 = Sequence()
   n3 = Get Anim_Inst() -> (Anim_Inst:Anim Instance Object Reference)
   > Is Valid(InputObject=@n3)
   > Branch(Condition=Get bIsInFrontEnd())
   > Set isIdleEmotePlaying(isIdleEmotePlaying:double="To Float (Boolean)"(InBool=Montage_IsPlaying(self=@n3, Montage=/Game/Animation/Game/MainPlayer/Menu/BR/Male_Commando_Idle_01_M))) -> (Output_Get:double)
   n4 = BreakTransform(InTransform=GetDeltaTransformFromRefPose(self=GetOwnerSkeletalMeshComponent(), BoneName=thigh_l, BaseName=pelvis))
   Set L_thighRotationAlpha(execute=@n2.then_1, L_thighRotationAlpha:double=MapRangeClamped(Value=BreakRotator(InRot=@n4.Rotation).Yaw, InRangeA=-35, InRangeB=-70)) -> (Output_Get:double)
   > Set isSkydiving(isSkydiving:double=SelectFloat(A=0.8, B=0, bPickA=Get bIsSkyDiving())) -> (Output_Get:double)
   > Set skydivingFinal(skydivingFinal:double=SelectFloat(A=0.6, B=Get isSkydiving() -> (isSkydiving:double), bPickA=Get bIsParachuteOpened())) -> (Output_Get:double)
   n5 = NOT Boolean(A=Get bIsAccelerating2D())
   n6 = Get bIsCrouching()
   > Set Is Crouching Not Moving(Is Crouching Not Moving:double="To Float (Boolean)"(InBool=AND Boolean(A=@n6, B=@n5))) -> (Output_Get:double)
   n7 = Get bRigidBodyEnabled()
   n8 = Get bIsAccelerating2D()
   n9 = Get bIsCrouching()
   > Set isCrouchMoving(isCrouchMoving:double="To Float (Boolean)"(InBool=AND Boolean!(A=@n9, B=@n8, C:bool=@n7) -> (ReturnValue))) -> (Output_Get:double)
   > Set pocketAlpha(pocketAlpha:double=SelectFloat(A=0.5, B=1, bPickA=Get bIsCrouching())) -> (Output_Get:double)
  ]}
]
variables [
 Anim_Inst {type=AnimInstance, default=null}
 Is Crouching Not Moving {type=number}
 isCrouchMoving {type=number}
 isIdleEmotePlaying {type=number}
 isSkydiving {type=number}
 L_thighRotationAlpha {type=number}
 pocketAlpha {type=number}
 Skel_Mesh {type=SkeletalMeshComponent}
 skydivingFinal {type=number}
]
